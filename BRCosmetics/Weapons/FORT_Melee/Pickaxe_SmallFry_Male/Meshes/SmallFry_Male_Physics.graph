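# Physics asset for the SmallFry Male pickaxe, a two-handed melee weapon cosmetic with a dynamic olive joint.
# A long thin capsule covers the handle bone and a shorter, wider capsule covers the olive joint bone.
# A single constraint ties the olive joint to the handle with twist and one swing axis locked, so rotation stays on the remaining swing axis.
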PhysicsAsset /BRCosmetics/Weapons/FORT_Melee/Pickaxe_SmallFry_Male/Meshes/SmallFry_Male_Physics
bodies {
 Melee_TwoHanded_Handle [{shapeType=Capsule, length=50, radius=2, center=<0 -0.8842 13.3007>, rotation=<rot 0 -90 0>}]
 dyn_olive_jnt [{shapeType=Capsule, length=30, radius=8, center=<-0.1006 0 0.8655>, rotation=<rot -87.7144 0 0>}]
}
constraints [{
  bone1Name=dyn_olive_jnt
  bone2Name=Melee_TwoHanded_Handle
  swing1LimitDegrees=10
  swing2LimitDegrees=45
  swing2Motion=Locked
  twistLimitDegrees=45
  twistMotion=Locked
 }]
